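; BP_WaterLineDecal is a Blueprint actor that sets up a water line decal using the M_WaterLineDecal material.
; Its construction script creates a dynamic material instance and assigns it to the Decal component.
; It multiplies the decal size by the actor scale and passes the result as the DecalInputs vector parameter, so the material receives the decal's scaled dimensions.
; BeginPlay and Tick are empty.
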(event EventBeginPlay)

(event EventTick
  (params (DeltaSeconds :type Float%20%28single-precision%29)))

(fn ConstructionScript
  (bind _returnvalue (Rendering|Material|CreateDynamicMaterialInstance "/Game/Environments/TechArt/WaterLineDecal/M_WaterLineDecal.M_WaterLineDecal" "None" "None" :!class /Script/Engine.KismetMaterialLibrary))
  (Rendering|Material|SetVectorParameterValue _returnvalue "DecalInputs" (Math|Conversions|ToLinearColor(Vector) (* (Variables|/Script/Engine.DecalComponent|GetDecalSize (Variables|Decal|GetDecal)) (Transformation|GetActorScale3D))) :!class /Script/Engine.MaterialInstanceDynamic)
  (Rendering|Components|Decal|SetDecalMaterial (Variables|Decal|GetDecal) _returnvalue))
