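# Translucent deferred decal material for a water line, fading upward from the decal origin's height.
# DecalColor packs per-decal data: R and G each hold two values (floor/10 and fraction) blended by a power-shaped height gradient, B drives gradient height and, through its fraction, edge falloff, and A scales the mask.
# That mask multiplies the inverted height gradient by a distance-based soft XY box fade sized from DecalInputs and SizeFactor.
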Material /Game/Environments/TechArt/WaterLineDecal/M_WaterLineDecal
graph {
 properties {bAutomaticallySetUsageInEditor=false, blendMode=BLEND_Translucent, materialDomain=MD_DeferredDecal, usedShadingModels=Default Lit}
 nodes [
  Constant()
  n1 = DecalColor()
  n2 = WorldPosition()
  n3 = ObjectPositionWS()
  n4 = Divide(A=Subtract(A=@n2.Z, B=ComponentMask(None=@n3) {b=true}), B=@n1.B) {constB=50}
  n5 = Saturate(None=@n4)
  n6 = OneMinus(None=Power(Base=OneMinus(None=@n5), Exp=ScalarParameter() {defaultValue=4, parameterName=CenterTransitionPow}) {constExponent=48})
  LinearInterpolate(A=Divide(A=Floor(None=@n1.R)) {constB=10}, B=Frac(None=@n1.R), Alpha=@n6)
  LinearInterpolate(A=Divide(A=Floor(None=@n1.G)) {constB=10}, B=Frac(None=@n1.G), Alpha=@n6)
  n7 = Saturate(None=@n4)
  n8 = Subtract(A=@n2.XY, B=ComponentMask(None=@n3) {g=true, r=true})
  MaterialExpressionVectorParameter_0 = VectorParameter() {parameterName=DecalInputs, defaultValue=<rgba 0 0 0 1>}
  n9 = Convert(None=@MaterialExpressionVectorParameter_0.RGB) {
   convertInputs [{
     type=Vector3
     expressionInput {
      expression=&self:MaterialExpressionVectorParameter_0
      inputName=None
      mask=1
      maskA=0
      maskB=1
      maskG=1
      maskR=1
      outputIndex=0
     }
    }]
   convertOutputs [{type=Vector2}]
  }
  n10 = Multiply(A=OneMinus(None=Frac(None=@n1.B)), B=ScalarParameter() {defaultValue=2000, parameterName=BoxFalloff})
  n11 = Subtract(A=Multiply(A=@n9, B=ScalarParameter() {defaultValue=2, parameterName=SizeFactor}), B=Multiply(A=@n10) {constB=2})
  n12 = Max(A=Subtract(A=Abs(None=@n8), B=Multiply(A=@n11) {constB=0.5})) {constB=0}
  n13 = Constant()
  n14 = Saturate(None=Divide(A=Distance(A=@n12, B=@n13), B=@n10))
  Multiply(A=@n1.A, B=Multiply(A=OneMinus(None=@n7), B=OneMinus(None=@n14)))
  ScalarParameter() {defaultValue=100, parameterName=GradientSize}
 ]
}
